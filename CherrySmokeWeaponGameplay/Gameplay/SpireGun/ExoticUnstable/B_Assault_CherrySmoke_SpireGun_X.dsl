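; Blueprint for an exotic "unstable" Spire assault weapon that varies its bullet spread pattern as it consumes ammo.
; BeginPlay shuffles a bag of indices, applies the first pattern and binds to OnConsumeAmmo.
; Ammo events step a DoN counter through the bag, selecting among ten square, rhombus, diagonal, horizontal and vertical variants to overwrite the weapon's BulletPattern.
; The bag reshuffles when exhausted, a shuffle-bag randomizer, with a custom reset event for the counter.
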(event EventBeginPlay
  (Default|ShuffleBagOfMarbles)
  (bind _selectedbulletpattern (Default|SelectPatternfromIndex 0))
  (Default|ResetandSetBulletPattern _selectedbulletpattern)
  (EventDispatchers|BindEventtoOnConsumeAmmo (delegate Custom|OnConsumeAmmo_Event)))

(event Custom|OnConsumeAmmo_Event
  (params (Quantity :type Integer))
  (Utilities|FlowControl|DoOnce false
    (:Completed
      (bind _counter (Utilities|FlowControl|DoN (Utilities|Array|Length (Variables|Default|GetBagOfMarbles)))
        (:Exit
          (bind _selectedbulletpattern (Default|SelectPatternfromIndex _counter))
          (Default|ResetandSetBulletPattern _selectedbulletpattern)
          (CallFunction|ResetCounterEvent)
          (if (<= (Utilities|Array|Length (Variables|Default|GetBagOfMarbles)) _counter)
            (Default|ShuffleBagOfMarbles)))))))

(event Custom|ResetCounter_Event
  (Utilities|FlowControl|DoOnce false
    (:Completed
      (bind _counter (Utilities|FlowControl|DoN (Utilities|Array|Length (Variables|Default|GetBagOfMarbles)))
        (:Exit
          (bind _selectedbulletpattern (Default|SelectPatternfromIndex _counter))
          (Default|ResetandSetBulletPattern _selectedbulletpattern)
          (CallFunction|ResetCounterEvent)
          (if (<= (Utilities|Array|Length (Variables|Default|GetBagOfMarbles)) _counter)
            (Default|ShuffleBagOfMarbles)))))))

(fn OnRep_BagOfMarbles)

(fn ResetAndSetBulletPattern
  (params (PatternToSet :type Array%20of%20Fort%20Bullet%20Pattern%20Entry%20Structures))
  (bind _bulletpattern (Variables|FortWeapon|GetBulletPattern))
  (Utilities|Array|Clear _bulletpattern)
  (Utilities|Array|AppendArray _bulletpattern PatternToSet))

(fn SelectPatternFromIndex
  (params (InputPin :type Integer))
  (return (Utilities|Select :Index (Utilities|Array|Get(aref) (Variables|Default|GetBagOfMarbles) InputPin) :Option%200 (Variables|Default|GetBulletPatternSquare1) :Option%201 (Variables|Default|GetBulletPatternSquare2) :Option%202 (Variables|Default|GetBulletPatternRhombus1) :Option%203 (Variables|Default|GetBulletPatternRhombus2) :Option%204 (Variables|Default|GetBulletPatternDiagonal1) :Option%205 (Variables|Default|GetBulletPatternDiagonal2) :Option%206 (Variables|Default|GetBulletPatternHorizontal1) :Option%207 (Variables|Default|GetBulletPatternHorizontal2) :Option%208 (Variables|Default|GetBulletPatternVertical1) :Option%209 (Variables|Default|GetBulletPatternVertical2)) 0))

(fn ShuffleBagOfMarbles
  (Utilities|Array|Shuffle (Variables|Default|GetBagOfMarbles)))

(fn ConstructionScript
  (|CallParentFunction))
